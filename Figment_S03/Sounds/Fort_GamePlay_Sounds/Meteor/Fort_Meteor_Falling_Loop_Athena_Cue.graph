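# Sound cue for a falling meteor in Athena, mixing three attenuated layers that each pick randomly between two waves.
# Distant and very distant looping layers get random pitch variation and use separate attenuation assets, with the very distant one mixed much quieter.
# A non-looping "almost landed" layer uses an inline attenuation override.
# The cue plays in the events sound class, with concurrency capped by stopping the quietest instance.
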SoundCue /Figment_S03/Sounds/Fort_GamePlay_Sounds/Meteor/Fort_Meteor_Falling_Loop_Athena_Cue
root=Mixer_2
nodes [
 {
  id=Attenuation_0
  children [Modulator_0]
  properties {attenuationSettings=&/Fort_Meteor_Athena_Att}
 }
 {
  id=Attenuation_1
  children [Random_1]
  properties {bOverrideAttenuation=true, attenuationOverrides {falloffDistance=3500, stereoSpread=1000, attenuationShapeExtents {x=3000}}}
 }
 {
  id=Attenuation_2
  children [Modulator_1]
  properties {attenuationSettings=&/Fort_Meteor_Distant_Athena_Att}
 }
 {id=Mixer_2, children [Attenuation_0, Attenuation_2, Attenuation_1], properties {inputVolume [0.75,0.1,0.6]}}
 {id=Modulator_0, children [Random_0], properties {pitchMax=1.2, pitchMin=0.9}}
 {id=Modulator_1, children [Random_2], properties {pitchMax=1.2, pitchMin=0.9}}
 {id=Random_0, children [WavePlayer_4,WavePlayer_5], properties {weights [1,1]}}
 {id=Random_1, children [WavePlayer_7,WavePlayer_8], properties {weights [1,1]}}
 {id=Random_2, children [WavePlayer_6,WavePlayer_9], properties {weights [1,1]}}
 {id=WavePlayer_4, properties {bLooping=true, soundWaveAssetPtr=&/Meteor_Falling_Distant_01}}
 {id=WavePlayer_5, properties {bLooping=true, soundWaveAssetPtr=&/Meteor_Falling_Distant_02}}
 {
  id=WavePlayer_6
  properties {bLooping=true, soundWaveAssetPtr=&/Meteor_Falling_VeryDistant_lp_01}
 }
 {id=WavePlayer_7, properties {soundWaveAssetPtr=/Game/Sounds/Fort_GamePlay_Sounds/Meteor/Meteor_Almost_Landed_01}}
 {id=WavePlayer_8, properties {soundWaveAssetPtr=/Game/Sounds/Fort_GamePlay_Sounds/Meteor/Meteor_Almost_Landed_02}}
 {
  id=WavePlayer_9
  properties {bLooping=true, soundWaveAssetPtr=&/Meteor_Falling_VeryDistant_lp_02}
 }
]
properties {
 bOverrideConcurrency=true
 maxDistance=22000
 priority=20
 soundClassObject=&Fort_Events
 volumeMultiplier=0.5
 concurrencyOverrides {maxCount=2, resolutionRule=StopQuietest}
}
