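# Sound cue for a meteor landing in the Athena game mode, layering a close impact with a distant landing sound.
# The close layer randomly picks one of two impact waves and the distant layer plays a single wave, each pitch-modulated and given its own attenuation asset.
# A mixer sums both layers under the Fort_Events sound class, with raised priority and a concurrency override capping simultaneous instances.
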SoundCue /Figment_S03/Sounds/Fort_GamePlay_Sounds/Meteor/Fort_Meteor_Land_Athena_Cue
root=Mixer_2
nodes [
 {
  id=Attenuation_4
  children [Modulator_1]
  properties {attenuationSettings=&/Fort_Meteor_Athena_Att}
 }
 {
  id=Attenuation_5
  children [Modulator_2]
  properties {attenuationSettings=&/Fort_Meteor_Distant_Athena_Att}
 }
 {id=Mixer_2, children [SoundClass_2,SoundClass_1], properties {inputVolume [1,1.1]}}
 {id=Modulator_1, children [Random_0], properties {pitchMax=1.2, pitchMin=0.9}}
 {id=Modulator_2, children [WavePlayer_3], properties {pitchMax=1.2, pitchMin=0.9}}
 {id=Random_0, children [WavePlayer_4,WavePlayer_5], properties {weights [1,1]}}
 {id=SoundClass_1, children [Attenuation_5], properties {soundClassOverride=&Fort_Events}}
 {id=SoundClass_2, children [Attenuation_4], properties {soundClassOverride=&Fort_Events}}
 {id=WavePlayer_3, properties {soundWaveAssetPtr=&/Meteor_Land_Distant}}
 {id=WavePlayer_4, properties {soundWaveAssetPtr=&/Meteor_Impact_01}}
 {id=WavePlayer_5, properties {soundWaveAssetPtr=&/Meteor_Impact_02}}
]
properties {
 bOverrideConcurrency=true
 maxDistance=22000
 priority=30
 soundClassObject=&Fort_Events
 volumeMultiplier=0.5
 concurrencyOverrides {maxCount=2}
}
